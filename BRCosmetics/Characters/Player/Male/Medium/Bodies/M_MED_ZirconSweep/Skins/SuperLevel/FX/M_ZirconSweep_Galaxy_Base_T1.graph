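# Translucent, unlit, two-sided FX material for the ZirconSweep SuperLevel skin's galaxy base, usable on Niagara mesh particles.
# Time-of-day-scaled colors blend by a V gradient plus panning swirl texture, feeding an emissive adjustment function.
# A mask from V gradients and cloud texture is shaped by two-sided Fresnel, apparently for opacity.
# A WPO branch combines panning-noise displacement, faded along inverted V, with optional normal-based expansion.
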Material /BRCosmetics/Characters/Player/Male/Medium/Bodies/M_MED_ZirconSweep/Skins/SuperLevel/FX/M_ZirconSweep_Galaxy_Base_T1
graph {
 properties {
  bUsedWithNiagaraMeshParticles=true
  blendMode=BLEND_Translucent
  pixelDepthOffsetMode=PDOM_Legacy
  shadingModel=MSM_Unlit
  twoSided=true
  usedShadingModels=Unlit
 }
 nodes [
  Clamp()
  n1 = ScalarParameter() {defaultValue=2.5, parameterName=ColorMult_Day}
  n2 = MaterialFunctionCall(Color_Morning=@n1, Color_Day=@n1, Color_Evening=@n1, Color_Night=ScalarParameter() {defaultValue=1.5, parameterName=ColorMult_Night}) {materialFunction=&Color_Time_of_Day}
  n3 = Multiply(A=@n2.Result, B=VectorParameter() {parameterName=ZirconSweep_Base_Color1, defaultValue=<rgba 1 0.0195 0 1>}.RGB)
  n4 = TextureCoordinate()
  n5 = Multiply(A=Power(Base=ComponentMask(None=@n4) {g=true}, Exp=Constant() {r=8}), B=Constant() {r=3})
  n6 = TextureCoordinate()
  n7 = AppendVector(A=ScalarParameter() {defaultValue=1}, B=ScalarParameter() {defaultValue=0.5, parameterName=Param_1})
  n8 = ScalarParameter() {defaultValue=-0.15, parameterName=Param_2}
  n9 = ScalarParameter() {defaultValue=0.05, parameterName=Param_3}
  n10 = TextureSample(UVs=Panner(Coordinate=Multiply(A=@n6, B=@n7), Speed=AppendVector(A=@n8, B=@n9))) {samplerType=SAMPLERTYPE_LinearColor, texture=/Game/Global/Textures/Default/Caustics/03_WaterFoam_AccentSwirls_H}
  n11 = LinearInterpolate(
   A=@n3
   B=Multiply(A=@n2.Result, B=VectorParameter() {parameterName=ZirconSweep_Base_Color2, defaultValue=<rgba 0.9115 0.0178 0 1>}.RGB)
   Alpha=Clamp(None=Add(A=@n5, B=@n10.RGB))
  )
  LinearInterpolate(A=@n3, B=@n11)
  MaterialFunctionCall(Emissive=@n11) {materialFunction=&MF_FX_EmissiveAdjustments}
  n12 = ComponentMask(None=TextureCoordinate()) {g=true}
  n13 = Constant() {r=8}
  n14 = Power(Base=ComponentMask(None=TextureCoordinate()) {g=true}, Exp=Constant() {r=2.2})
  n15 = Constant() {r=4}
  n16 = Multiply(A=TextureCoordinate(), B=AppendVector(A=Constant() {r=1}, B=Constant() {r=1}))
  n17 = Add(
   A=Multiply(A=Power(Base=@n12, Exp=@n13), B=Constant() {r=3})
   B=Multiply(
    A=Multiply(A=@n14, B=@n15)
    B=TextureSample(UVs=@n16) {texture=/Game/Packages/Fortress_Transylvania/Haunted_Forest/Swamps/Textures/clouds_dust_detail_01}.RGB
   )
  )
  n18 = Constant() {r=1}
  n19 = Constant() {r=1}
  n20 = VertexNormalWS()
  Multiply(
   A=Multiply(
    A=Clamp(None=@n17)
    B=MaterialFunctionCall(Brightness=@n18, Exponent=@n19, Vertex / Pixel Normal=@n20) {materialFunction=&MF_Fresnel_FX}.Fresnel Double Sided
   )
   B=ScalarParameter() {defaultValue=0.75, parameterName=B}
  )
  n21 = AppendVector(A=Constant(), B=Constant())
  n22 = LinearInterpolate(
   A=TextureCoordinate()
   B=TextureCoordinate() {coordinateIndex=1}
   Alpha=ScalarParameter() {group=WPO, parameterName=Use_UV0-or-UV1, sortPriority=1}
  )
  n23 = AppendVector(
   A=ScalarParameter() {defaultValue=1, group=WPO, parameterName=WPO_Tiling_X}
   B=ScalarParameter() {defaultValue=1, group=WPO, parameterName=WPO_Tiling_Y}
  )
  n24 = ScalarParameter() {defaultValue=-0.25, group=WPO, parameterName=WPO_Speed_X}
  n25 = ScalarParameter() {defaultValue=0.2, group=WPO, parameterName=WPO_Speed_Y}
  n26 = TextureSampleParameter2D(UVs=Panner(Coordinate=Multiply(A=@n22, B=@n23), Speed=AppendVector(A=@n24, B=@n25))) {group=WPO, parameterName=WPO Texture, texture=/Game/Characters/Player/Male/Medium/Bodies/M_MED_Viper/Textures/Glitch/T_SoftNoise}
  n27 = LinearInterpolate(
   B=Multiply(A=AppendVector(A=@n21, B=@n26.B), B=ScalarParameter() {defaultValue=1, group=WPO, parameterName=WPO_Intensity})
   Alpha=ScalarParameter() {defaultValue=1, group=WPO, parameterName=Use_WPO, sortPriority=-1}
  )
  n28 = ComponentMask(None=OneMinus(None=TextureCoordinate())) {g=true}
  n29 = Constant() {r=2}
  n30 = Clamp(None=Multiply(A=Power(Base=@n28, Exp=@n29), B=Constant() {r=8}))
  n31 = Divide(A=VertexNormalWS(), B=Constant() {r=10})
  n32 = ScalarParameter() {group=WPO, parameterName=WPO_VertexExpand_Mult, sortPriority=0}
  Reroute(None=Add(A=Multiply(A=@n27, B=@n30), B=Multiply(A=@n31, B=@n32)))
 ]
}
