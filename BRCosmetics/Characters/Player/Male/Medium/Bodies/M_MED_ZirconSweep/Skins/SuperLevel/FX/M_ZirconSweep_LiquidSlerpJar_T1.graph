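# Unlit translucent material for Niagara mesh particles, apparently the swirling liquid in the ZirconSweep skin's slurp jar effect.
# Panning cloud noise, distorted by noise and bump-offset by another scrolling noise layer, blends two inside colors.
# A time-of-day multiplier blends in along a vertical UV gradient, and a second gradient scales the result into emissive adjustments.
# A separate chain multiplies the noise by powered fresnel and a master opacity scalar.
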Material /BRCosmetics/Characters/Player/Male/Medium/Bodies/M_MED_ZirconSweep/Skins/SuperLevel/FX/M_ZirconSweep_LiquidSlerpJar_T1
graph {
 properties {
  bUsedWithNiagaraMeshParticles=true
  blendMode=BLEND_Translucent
  pixelDepthOffsetMode=PDOM_Legacy
  shadingModel=MSM_Unlit
  usedShadingModels=Unlit
 }
 nodes [
  n1 = ComponentMask(None=TextureCoordinate()) {g=true}
  n2 = ScalarParameter() {defaultValue=4, parameterName=Outside_Fade_Power}
  Clamp(None=Multiply(A=Power(Base=@n1, Exp=@n2), B=ScalarParameter() {defaultValue=2, parameterName=Outside_Fade_Mult}))
  n3 = TextureCoordinate()
  n4 = Power(Base=ComponentMask(None=@n3) {g=true}, Exp=ScalarParameter() {defaultValue=1, parameterName=Param_1})
  n5 = Constant() {r=1}
  MaterialExpressionRemap_1 = MaterialXRemap!(
   None=Clamp(None=Multiply(A=@n4, B=@n5))
   InputLow
   InputHigh
   TargetLow=ScalarParameter() {defaultValue=1, parameterName=TargetLow}
   TargetHigh=ScalarParameter() {defaultValue=0.1, parameterName=TargetHigh}
  ) {targetHighDefault=0.2, targetLowDefault=0.2}
  n6 = Constant() {r=0.2}
  n7 = Constant() {r=1}
  n8 = Panner(
   Coordinate=Multiply(A=TextureCoordinate(), B=AppendVector(A=@n6, B=@n7))
   Speed=AppendVector(
    A=ScalarParameter() {defaultValue=-0.0125, parameterName=Outside_Pan_X}
    B=ScalarParameter() {defaultValue=0.0125, parameterName=Outside_Pan_Y}
   )
  )
  n9 = TextureCoordinate()
  n10 = AppendVector(A=Constant() {r=0.2}, B=Constant() {r=1})
  n11 = ScalarParameter() {defaultValue=-0.0125, parameterName=Outside_Pan_X}
  n12 = ScalarParameter() {defaultValue=0.0125, parameterName=Outside_Pan_Y}
  n13 = TextureSample(UVs=Panner(Coordinate=Multiply(A=@n9, B=@n10), Speed=AppendVector(A=@n11, B=@n12))) {texture=&AtmosphericCloudNoise01}
  n14 = ScalarParameter() {defaultValue=1, parameterName=Outside_Noise_Distortion_Power}
  n15 = Clamp(None=Multiply(A=Power(Base=@n13.R, Exp=@n14), B=ScalarParameter() {defaultValue=0.2, parameterName=Outside_Noise_Distortion_Mult}))
  n16 = TextureCoordinate()
  n17 = AppendVector(A=Constant() {r=0.5}, B=Constant() {r=0.5})
  n18 = Constant() {r=-0.025}
  n19 = Constant() {r=0.025}
  n20 = TextureSample(UVs=Panner(Coordinate=Multiply(A=@n16, B=@n17), Speed=AppendVector(A=@n18, B=@n19))) {texture=&AtmosphericCloudNoise01}
  MaterialExpressionRemap_0 = MaterialXRemap(None=@n20.RGB) {targetLowDefault=0.5}
  n21 = ScalarParameter() {defaultValue=-7, parameterName=BumpOffset-Intensity}
  n22 = TextureSample(UVs=BumpOffset(Coordinate=Add(A=@n8, B=@n15), Height=Multiply(A=@MaterialExpressionRemap_0, B=@n21) {constB=-2})) {texture=&AtmosphericCloudNoise01}
  n23 = ScalarParameter() {defaultValue=0.5, parameterName=Outside_Texture_Power}
  n24 = Clamp(None=Multiply(A=Power(Base=@n22.RGB, Exp=@n23), B=ScalarParameter() {defaultValue=1, parameterName=Outside_Texture_Mult}))
  n25 = LinearInterpolate(
   A=VectorParameter() {parameterName=Inside_Color_A, defaultValue=<rgba 1 0.0195 0 1>}.RGB
   B=VectorParameter() {parameterName=Inside_Color_B, defaultValue=<rgba 0.8073 0.0157 0 1>}.RGB
   Alpha=@n24
  )
  n26 = ScalarParameter() {defaultValue=15, parameterName=ColorMult_Day}
  n27 = Multiply(
   A=MaterialFunctionCall(Color_Morning=@n26, Color_Day=@n26, Color_Evening=@n26, Color_Night=ScalarParameter() {defaultValue=4, parameterName=ColorMult_Night}) {materialFunction=&Color_Time_of_Day}.Result
   B=@n25
  )
  n28 = ComponentMask(None=OneMinus(None=TextureCoordinate())) {g=true}
  n29 = ScalarParameter() {defaultValue=1, parameterName=Param_2}
  n30 = Clamp(None=Multiply(A=Power(Base=@n28, Exp=@n29), B=Constant() {r=1}))
  MaterialFunctionCall(Emissive=Multiply(A=@MaterialExpressionRemap_1, B=LinearInterpolate(A=@n25, B=@n27, Alpha=@n30))) {materialFunction=&MF_FX_EmissiveAdjustments}
  n31 = Constant() {r=1}
  n32 = MaterialFunctionCall(Brightness=@n31, Exponent=@n31) {materialFunction=&MF_Fresnel_FX}
  n33 = ScalarParameter() {defaultValue=3, group=Deep Noise, parameterName=Deep_Fresnel_Power}
  n34 = Multiply(A=Multiply(A=Power(Base=@n32.Fresnel, Exp=@n33), B=ScalarParameter() {defaultValue=1, group=Deep Noise, parameterName=Deep_Fresnel_Mult}))
  Multiply(A=Multiply(A=@n24, B=Clamp(None=@n34)), B=ScalarParameter() {defaultValue=0.5, parameterName=Opacity_Master})
  OneMinus(None=@n20.RGB)
  OneMinus(None=@n3)
 ]
}
